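# Niagara weapon trail for the Noble Male pickaxe, drawing short-lived ribbons between skeletal mesh bone or socket pairs via AnimTrail modules.
# AirTrail001_Saber feeds three ribbon renderers, with width shaped by particle and emitter-age curves, random alpha, color curves and dynamic material parameters.
# Trail_Filler3 adds a simpler blue-gray ribbon on a different bone pair, narrowing and fading over its life.
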NiagaraSystem /BRCosmetics/Weapons/FORT_Melee/Pickaxe_Noble_Male/FX/NS_Pickaxe_Noble_Trail
dependencies {
 usedDataInterfaces [&NiagaraDataInterfaceCurve, &NiagaraDataInterfaceColorCurve]
 usedDynamicInputs []
 usedModules [&SystemState]
}
emitters {
 AirTrail001_Saber {
  emitterUpdateScript [
   &EmitterState(Life Cycle Mode=::NewEnumerator1, Loop Behavior=::NewEnumerator1, Loop Delay=0.05, Loop Duration=0.25, UseLoopDelay=false)
   &FN_SpawnAnimTrail_v2()
  ]
  particleSpawnScript [
   /Game/Effects/Niagara/Modules/AnimTrail/FN_InitializeAnimTrail/FN_InitializeAnimTrailBoneIndecies(First Bone or Socket Index=2, Second Bone or Socket Index=3)
   /Game/Effects/Niagara/Modules/AnimTrail/FN_InitializeAnimTrail/FN_InitializeAnimTrail(
    AnimTrailWidth=1.2
    Color=&MakeLinearColorFromVectorAndFloat(
     "Vector (RGB)"=&MakeVectorFromLinearColorRGB(Color=<rgba 1 1 1 1>)
     "Float (Alpha)"=&RandomRangeFloat(Minimum=0.02, Maximum=0.4)
    )
    Lifetime=0.2
   )
   /SetVariables() {enabled=false, isSetParameters=true}
  ]
  particleUpdateScript [
   &SolveForcesAndVelocity(Clamp Velocity=false, Limit Acceleration=false)
   &ScaleRibbonWidth(Ribbon Width Scale=&Multiply_Float(
     A=&FloatFromCurve(FloatCurve=DataInterface() {propertyValues {Curve {keys [
          {arriveTangent=-0.586, interpMode=RCIM_Cubic, leaveTangent=-0.586, tangentMode=RCTM_User, time=0, value=1}
          {arriveTangent=-0.9664, interpMode=RCIM_Cubic, leaveTangent=-0.9664, tangentMode=RCTM_User, time=0.5, value=0.9}
          <at 1 0>
         ]}}})
     B=&FloatFromCurve(
      FloatCurve=DataInterface() {propertyValues {Curve {keys [
          {interpMode=RCIM_Cubic, tangentMode=RCTM_User, time=0, value=2}
          {interpMode=RCIM_Cubic, tangentMode=RCTM_User, time=0.2771, value=0.3122}
          {tangentMode=RCTM_User, time=0.5055, value=0}
         ]}}}
      CurveIndex=$Emitter.Age
     )
    ))
   &FN_UpdateAnimTrail_v2()
   &DynamicMaterialParameters(
    Index 0 Param 1=&UniformRangedFloat(Minimum=0, Maximum=1)
    Index 0 Param 2=&UniformRangedFloat(Minimum=0, Maximum=1)
    Index 0 Param 3=1
    Index 0 Param 4=1
    Module.Param0WriteEnabled=true
    Module.Param1WriteEnabled=true
    Module.Param2WriteEnabled=true
    Module.Param3WriteEnabled=true
   )
   &ScaleColor(
    Scale Alpha=&FloatFromCurve(FloatCurve=DataInterface() {propertyValues {Curve {keys [
         {interpMode=RCIM_Cubic, tangentMode=RCTM_User, time=0, value=0}
         {interpMode=RCIM_Cubic, tangentMode=RCTM_User, time=0.2, value=1}
         {tangentMode=RCTM_User, time=1, value=0}
        ]}}})
    Scale RGB=&MakeVectorFromLinearColorRGB(Color=&ColorFromCurve(ColorCurve=DataInterface() {propertyValues {
        AlphaCurve {keys [
          {arriveTangent=-1.791, interpMode=RCIM_Cubic, leaveTangent=-1.791, time=0.4417, value=1}
          {arriveTangent=-1.791, interpMode=RCIM_Cubic, leaveTangent=-1.791, time=1, value=0}
         ]}
        BlueCurve {keys [<at 0 0.5>, <at 0.2752 1>, {arriveTangent=-33.3264, leaveTangent=-33.3264, time=0.6037, value=0.475}]}
        GreenCurve {keys [<at 0 0.4529>, <at 0.2752 0.925>, {arriveTangent=-10.1691, leaveTangent=-10.1691, time=0.6037, value=0.4394}]}
        RedCurve {keys [<at 0 0.45>, <at 0.2752 0.9>, {arriveTangent=-8.7497, leaveTangent=-8.7497, time=0.6037, value=0.4275}]}
       }}))
    ScaleA=true
    ScaleRGB=true
   )
   /SetVariables(Particles.ColorAlt:LinearColor=&MakeLinearColorFromVectorAndFloat(
     "Vector (RGB)"=&MakeVectorFromLinearColorRGB(Color=&ColorFromCurve(ColorCurve=DataInterface() {propertyValues {
         AlphaCurve {keys [<at 0 1 RCIM_Cubic>,<at 1 1 RCIM_Cubic>]}
         BlueCurve {keys [<at 0.0444 0.4 RCIM_Cubic>,<at 0.2695 0.15 RCIM_Cubic>]}
         GreenCurve {keys [<at 0.0444 0.37 RCIM_Cubic>,<at 0.2695 0.1387 RCIM_Cubic>]}
         RedCurve {keys [<at 0.0444 0.36 RCIM_Cubic>,<at 0.2695 0.135 RCIM_Cubic>]}
        }}))
     "Float (Alpha)"=20
    )) {isSetParameters=true}
  ]
  rendererClasses [&NiagaraRibbonRendererProperties]
  renderers [
   {rendererClass=&NiagaraRibbonRendererProperties}
   {rendererClass=&NiagaraRibbonRendererProperties, rendererIndex=1}
   {rendererClass=&NiagaraRibbonRendererProperties, rendererIndex=2}
  ]
 }
 Trail_Filler3 {
  emitterUpdateScript [
   &EmitterState(Life Cycle Mode=::NewEnumerator1, Loop Duration=0.4, Recalculate Duration Each Loop=true, UseLoopDelay=false)
   &FN_SpawnAnimTrail()
  ]
  particleSpawnScript [
   &InitializeParticle(UsePositionOffset=false, Write Material Random=true) {enabled=false}
   /Game/Effects/Niagara/Modules/AnimTrail/FN_InitializeAnimTrailBoneIndecies_v2/FN_InitializeAnimTrailBoneIndecies(First Bone or Socket Index=2, Second Bone or Socket Index=4)
   /Game/Effects/Niagara/Modules/AnimTrail/FN_InitializeAnimTrail/FN_InitializeAnimTrail_v5/FN_InitializeAnimTrail(AnimTrailWidth=1.1, Color=<rgba 0.45 0.4625 0.5 0.5>, Lifetime=0.2)
  ]
  particleUpdateScript [
   &ParticleState()
   &FN_UpdateAnimTrail(Kill Locked particle on system deactivate=false, SkeletalMesh=$User.SkeletalMesh)
   &ScaleRibbonWidth(Ribbon Width Scale=&FloatFromCurve(FloatCurve=DataInterface() {propertyValues {Curve {keys [
         {arriveTangent=-1.2579, interpMode=RCIM_Cubic, time=0, value=1}
         {arriveTangent=-0.9664, interpMode=RCIM_Cubic, leaveTangent=-0.9664, tangentMode=RCTM_User, time=1, value=0}
        ]}}}))
   &ScaleColor(
    Scale Alpha=&FloatFromCurve(FloatCurve=DataInterface() {propertyValues {Curve {keys [<at 0 1>,<at 1 0>]}}})
    ScaleA=true
    ScaleRGB=true
   )
  ]
  rendererClasses [&NiagaraRibbonRendererProperties]
  renderers [{rendererClass=&NiagaraRibbonRendererProperties}]
 }
}
summary {
 systemName=NS_Pickaxe_Noble_Trail
 emitters [
  {emitterName=Trail_Filler3, rendererClasses [&NiagaraRibbonRendererProperties]}
  {emitterName=AirTrail001_Saber, rendererClasses [&NiagaraRibbonRendererProperties]}
 ]
 userVariables [
  User.bIsFrontEndPreview {type=NiagaraBool}
  User.Color1Blue {type=LinearColor, description="", defaultValue {struct=LinearColor, value=<rgba 1 1 1 1>}}
  User.Color2Green {type=LinearColor, description="", defaultValue {struct=LinearColor, value=<rgba 0.4968 0.519 0.552 1>}}
  User.RibbonIDOffset {type=NiagaraInt32}
  User.SkeletalMesh {
   type=NiagaraDataInterfaceSkeletalMesh
   defaultValue {
    struct=NiagaraExt_VariableValue_DataInterface
    value {dataInterface=&self:NiagaraDataInterfaceSkeletalMesh_0}
   }
  }
  User.VariantWhite {type=NiagaraFloat, defaultValue=0, description=""}
 ]
}
